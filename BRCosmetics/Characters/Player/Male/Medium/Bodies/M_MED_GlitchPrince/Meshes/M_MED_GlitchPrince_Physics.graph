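# Physics asset for the M_MED_GlitchPrince player body, set up for secondary motion on the jacket.
# Six jacket chains (left/right front, side, back) run master-01-02 from spine_03, with limited lower joints and free master joints except the limited left-front one.
# Capsules on pelvis, spine and upper thighs form the body volume.
# Free constraints also cross-link some neighboring tips: back-to-back, left side-to-front, right back-to-side.
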PhysicsAsset /BRCosmetics/Characters/Player/Male/Medium/Bodies/M_MED_GlitchPrince/Meshes/M_MED_GlitchPrince_Physics
bodies {
 dyn_jacket_l_01 [
  {shapeType=Capsule, length=9.4396, radius=2.497, shapeName=dyn_skirt_side_03_l_capsule, center=<6.666 -0.5633 2>, rotation=<rot 0 0 90>}
  {shapeType=Capsule, length=9.4396, radius=2.497, shapeName=dyn_skirt_side_03_l_capsule, center=<2.2108 -0.5633 2>, rotation=<rot 0 0 90>}
 ]
 dyn_jacket_l_02 [
  {shapeType=Capsule, length=19.3684, radius=5.8536, shapeName=dyn_skirt_side_03_l_capsule, center=<12.7564 -0.5633 4>, rotation=<rot 0 0 90>}
  {shapeType=Capsule, length=19.7264, radius=6.3927, shapeName=dyn_skirt_side_03_l_capsule, center=<5.8713 -0.5633 4>, rotation=<rot 0 0 90>}
 ]
 dyn_jacket_l_bk_01 [
  {shapeType=Capsule, length=9.2213, radius=2.7153, shapeName=dyn_skirt_side_03_l_capsule, center=<6.666 -0.5633 2>, rotation=<rot 0 0 90>}
  {shapeType=Capsule, length=9.2213, radius=2.7153, shapeName=dyn_skirt_side_03_l_capsule, center=<2.2108 -0.5633 2>, rotation=<rot 0 0 90>}
 ]
 dyn_jacket_l_bk_02 [
  {shapeType=Capsule, length=12.7882, radius=5.6012, shapeName=dyn_skirt_side_03_l_capsule, center=<12.7564 -0.5633 3>, rotation=<rot 0 0 90>}
  {shapeType=Capsule, length=7.3354, radius=4.6012, shapeName=dyn_skirt_side_03_l_capsule, center=<5.8713 0 3>, rotation=<rot 0 0 90>}
  {
   shapeType=Capsule
   length=7.5937
   radius=4.3429
   shapeName=dyn_jacket_l_bk_02_capsule
   center=<11.7886 6.5373 0.5377>
   rotation=<rot 0 0 119.2979>
  }
 ]
 dyn_jacket_l_bk_master [{shapeType=Capsule, length=9.9366, radius=2, shapeName=dyn_skirt_side_03_l_capsule, rotation=<rot 0 0 90>}]
 dyn_jacket_l_fr_01 [
  {shapeType=Capsule, length=9.9366, radius=2, shapeName=dyn_skirt_side_03_l_capsule, center=<6.666 -0.5633 2>, rotation=<rot 0 0 90>}
  {shapeType=Capsule, length=9.9366, radius=2, shapeName=dyn_skirt_side_03_l_capsule, center=<2.2108 -0.5633 2>, rotation=<rot 0 0 90>}
 ]
 dyn_jacket_l_fr_02 [
  {shapeType=Capsule, length=14.1343, radius=3.518, shapeName=dyn_skirt_side_03_l_capsule, center=<8.948 3.7893 1.1978>, rotation=<rot 0 90 90>}
  {shapeType=Capsule, length=14.3909, radius=3.0041, shapeName=dyn_skirt_side_03_l_capsule, center=<9.0789 -0.5633 2>, rotation=<rot 0 90 90>}
  {
   shapeType=Capsule
   length=14.3438
   radius=3.0513
   shapeName=dyn_jacket_l_fr_02_capsule
   center=<9.1493 -6.6526 -0.4053>
   rotation=<rot 0 90 90>
  }
  {
   shapeType=Capsule
   length=10.6879
   radius=2.8414
   shapeName=dyn_jacket_l_fr_02_capsule
   center=<17.8694 -10.3639 -2.0765>
   rotation=<rot -0.0001 -179.6762 115.1071>
  }
 ]
 dyn_jacket_l_fr_master [{shapeType=Capsule, length=9.9366, radius=2, shapeName=dyn_skirt_side_03_l_capsule, rotation=<rot 0 0 90>}]
 dyn_jacket_l_master [{shapeType=Capsule, length=9.9366, radius=2, shapeName=dyn_skirt_side_03_l_capsule, rotation=<rot 0 0 90>}]
 dyn_jacket_r_01 [
  {shapeType=Capsule, length=9.9366, radius=2, shapeName=dyn_skirt_side_03_l_capsule, center=<6.666 -0.5633 2>, rotation=<rot 0 0 90>}
  {shapeType=Capsule, length=9.9366, radius=2, shapeName=dyn_skirt_side_03_l_capsule, center=<2.2108 -0.5633 2>, rotation=<rot 0 0 90>}
 ]
 dyn_jacket_r_02 [
  {
   shapeType=Capsule
   length=25.7271
   radius=6.4078
   shapeName=dyn_skirt_side_03_l_capsule
   center=<12.7564 -0.5633 1.2768>
   rotation=<rot 0 0 90>
  }
  {
   shapeType=Capsule
   length=21.0513
   radius=5.7681
   shapeName=dyn_skirt_side_03_l_capsule
   center=<5.8713 -0.5633 1.6463>
   rotation=<rot 0 0 90>
  }
 ]
 dyn_jacket_r_bk_01 [
  {shapeType=Capsule, length=9.2213, radius=2.7153, shapeName=dyn_skirt_side_03_l_capsule, center=<6.666 -0.5633 2>, rotation=<rot 0 0 90>}
  {shapeType=Capsule, length=9.2213, radius=2.7153, shapeName=dyn_skirt_side_03_l_capsule, center=<2.2108 -0.5633 2>, rotation=<rot 0 0 90>}
 ]
 dyn_jacket_r_bk_02 [
  {shapeType=Capsule, length=12.7882, radius=5.6012, shapeName=dyn_skirt_side_03_l_capsule, center=<12.7564 -0.5633 3>, rotation=<rot 0 0 90>}
  {shapeType=Capsule, length=7.3354, radius=4.6012, shapeName=dyn_skirt_side_03_l_capsule, center=<5.8713 -0.5633 3>, rotation=<rot 0 0 90>}
 ]
 dyn_jacket_r_bk_master [{shapeType=Capsule, length=9.9366, radius=2, shapeName=dyn_skirt_side_03_l_capsule, rotation=<rot 0 0 90>}]
 dyn_jacket_r_fr_01 [
  {shapeType=Capsule, length=9.9366, radius=2, shapeName=dyn_skirt_side_03_l_capsule, center=<6.666 -0.5633 2>, rotation=<rot 0 0 90>}
  {shapeType=Capsule, length=9.9366, radius=2, shapeName=dyn_skirt_side_03_l_capsule, center=<2.2108 -0.5633 2>, rotation=<rot 0 0 90>}
 ]
 dyn_jacket_r_fr_02 [
  {
   shapeType=Capsule
   length=11.3857
   radius=6.0074
   shapeName=dyn_skirt_side_03_l_capsule
   center=<8.948 7.6206 1.9227>
   rotation=<rot 0 90 90>
  }
  {shapeType=Capsule, length=11.9406, radius=5.4545, shapeName=dyn_skirt_side_03_l_capsule, center=<9.0789 -0.5633 2>, rotation=<rot 0 90 90>}
  {shapeType=Capsule, length=8, radius=5, shapeName=dyn_jacket_r_fr_02_capsule, center=<15.939 3.0021 -0.3729>, rotation=<rot 0 0 -91.8222>}
 ]
 dyn_jacket_r_fr_master [{shapeType=Capsule, length=9.9366, radius=2, shapeName=dyn_skirt_side_03_l_capsule, rotation=<rot 0 0 90>}]
 dyn_jacket_r_master [{shapeType=Capsule, length=9.9366, radius=2, shapeName=dyn_skirt_side_03_l_capsule, rotation=<rot 0 0 90>}]
 pelvis [
  {shapeType=Capsule, length=7.1728, radius=14.0612, shapeName=spine_01_capsule}
  {shapeType=Capsule, length=10.1728, radius=11.0612, shapeName=pelvis_capsule, center=<-13.9922 0 0>}
 ]
 spine_01 [{shapeType=Capsule, length=7.0678, radius=13.105, shapeName=spine_01_capsule, center=<0 -1.5766 0>}]
 spine_02 [{shapeType=Capsule, length=6.6774, radius=12.605, shapeName=spine_02_capsule, center=<0 -1.6285 0>}]
 spine_03 [{shapeType=Capsule, length=8.6774, radius=10.605, shapeName=spine_03_capsule}]
 thigh_twist_01_l [
  {shapeType=Capsule, length=29.1786, radius=10.1432, shapeName=thigh_twist_01_l_capsule, center=<-2.8362 0 0>, rotation=<rot 90 0 0>}
  {shapeType=Capsule, length=14.0687, radius=10.1432, shapeName=thigh_twist_01_l_capsule, center=<3.6886 -1.3977 4.8635>, rotation=<rot 90 0 0>}
 ]
 thigh_twist_01_r [{shapeType=Capsule, length=29.1786, radius=10.1432, shapeName=thigh_twist_01_l_capsule, center=<3.1585 0 0>, rotation=<rot 90 0 0>}]
}
constraints [
 {
  bone1Name=thigh_twist_01_l
  bone2Name=pelvis
  swing1LimitDegrees=45
  swing1Motion=Free
  swing2LimitDegrees=45
  swing2Motion=Free
  twistLimitDegrees=45
  twistMotion=Free
 }
 {
  bone1Name=dyn_jacket_l_bk_master
  bone2Name=spine_03
  swing1LimitDegrees=45
  swing1Motion=Free
  swing2LimitDegrees=45
  swing2Motion=Free
  twistLimitDegrees=45
  twistMotion=Free
 }
 {bone1Name=dyn_jacket_l_bk_01, bone2Name=dyn_jacket_l_bk_master, swing1LimitDegrees=10, swing2LimitDegrees=20, twistLimitDegrees=5}
 {bone1Name=dyn_jacket_l_bk_02, bone2Name=dyn_jacket_l_bk_01, swing1LimitDegrees=10, swing2LimitDegrees=30, twistLimitDegrees=5}
 {bone1Name=dyn_jacket_l_fr_master, bone2Name=spine_03, swing1LimitDegrees=45, swing2LimitDegrees=45, twistLimitDegrees=20}
 {bone1Name=dyn_jacket_l_fr_01, bone2Name=dyn_jacket_l_fr_master, swing1LimitDegrees=10, swing2LimitDegrees=45, twistLimitDegrees=5}
 {bone1Name=dyn_jacket_l_fr_02, bone2Name=dyn_jacket_l_fr_01, swing1LimitDegrees=10, swing2LimitDegrees=45, twistLimitDegrees=5}
 {
  bone1Name=dyn_jacket_l_master
  bone2Name=spine_03
  swing1LimitDegrees=45
  swing1Motion=Free
  swing2LimitDegrees=45
  swing2Motion=Free
  twistLimitDegrees=45
  twistMotion=Free
 }
 {bone1Name=dyn_jacket_l_01, bone2Name=dyn_jacket_l_master, swing1LimitDegrees=15, swing2LimitDegrees=20, twistLimitDegrees=20}
 {bone1Name=dyn_jacket_l_02, bone2Name=dyn_jacket_l_01, swing1LimitDegrees=20, swing2LimitDegrees=30, twistLimitDegrees=20}
 {
  bone1Name=dyn_jacket_r_bk_master
  bone2Name=spine_03
  swing1LimitDegrees=45
  swing1Motion=Free
  swing2LimitDegrees=45
  swing2Motion=Free
  twistLimitDegrees=45
  twistMotion=Free
 }
 {bone1Name=dyn_jacket_r_bk_01, bone2Name=dyn_jacket_r_bk_master, swing1LimitDegrees=10, swing2LimitDegrees=20, twistLimitDegrees=5}
 {bone1Name=dyn_jacket_r_bk_02, bone2Name=dyn_jacket_r_bk_01, swing1LimitDegrees=10, swing2LimitDegrees=30, twistLimitDegrees=5}
 {
  bone1Name=dyn_jacket_r_fr_master
  bone2Name=spine_03
  swing1LimitDegrees=45
  swing1Motion=Free
  swing2LimitDegrees=45
  swing2Motion=Free
  twistLimitDegrees=45
  twistMotion=Free
 }
 {bone1Name=dyn_jacket_r_fr_01, bone2Name=dyn_jacket_r_fr_master, swing1LimitDegrees=10, swing2LimitDegrees=45, twistLimitDegrees=5}
 {bone1Name=dyn_jacket_r_fr_02, bone2Name=dyn_jacket_r_fr_01, swing1LimitDegrees=10, swing2LimitDegrees=45, twistLimitDegrees=5}
 {
  bone1Name=dyn_jacket_r_master
  bone2Name=spine_03
  swing1LimitDegrees=45
  swing1Motion=Free
  swing2LimitDegrees=45
  swing2Motion=Free
  twistLimitDegrees=45
  twistMotion=Free
 }
 {bone1Name=dyn_jacket_r_01, bone2Name=dyn_jacket_r_master, swing1LimitDegrees=5, swing2LimitDegrees=20, twistLimitDegrees=10}
 {bone1Name=dyn_jacket_r_02, bone2Name=dyn_jacket_r_01, swing1LimitDegrees=5, swing2LimitDegrees=30, twistLimitDegrees=10}
 {
  bone1Name=spine_01
  bone2Name=pelvis
  swing1LimitDegrees=45
  swing1Motion=Free
  swing2LimitDegrees=45
  swing2Motion=Free
  twistLimitDegrees=45
  twistMotion=Free
 }
 {
  bone1Name=spine_02
  bone2Name=spine_01
  swing1LimitDegrees=45
  swing1Motion=Free
  swing2LimitDegrees=45
  swing2Motion=Free
  twistLimitDegrees=45
  twistMotion=Free
 }
 {
  bone1Name=spine_03
  bone2Name=spine_02
  swing1LimitDegrees=45
  swing1Motion=Free
  swing2LimitDegrees=45
  swing2Motion=Free
  twistLimitDegrees=45
  twistMotion=Free
 }
 {
  bone1Name=thigh_twist_01_r
  bone2Name=pelvis
  swing1LimitDegrees=45
  swing1Motion=Free
  swing2LimitDegrees=45
  swing2Motion=Free
  twistLimitDegrees=45
  twistMotion=Free
 }
 {
  bone1Name=dyn_jacket_r_bk_02
  bone2Name=dyn_jacket_l_bk_02
  swing1LimitDegrees=45
  swing1Motion=Free
  swing2LimitDegrees=45
  swing2Motion=Free
  twistLimitDegrees=45
  twistMotion=Free
 }
 {
  bone1Name=dyn_jacket_l_02
  bone2Name=dyn_jacket_l_fr_02
  swing1LimitDegrees=45
  swing1Motion=Free
  swing2LimitDegrees=45
  swing2Motion=Free
  twistLimitDegrees=45
  twistMotion=Free
 }
 {
  bone1Name=dyn_jacket_r_bk_02
  bone2Name=dyn_jacket_r_02
  swing1LimitDegrees=45
  swing1Motion=Free
  swing2LimitDegrees=45
  swing2Motion=Free
  twistLimitDegrees=45
  twistMotion=Free
 }
]
