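# Physics asset for the GlitchPrince male medium body's pants, set up for secondary motion on the belt and pant-leg bones.
# Each leg has four Dyn capsules, each tightly swing-limited with minimal twist to a small base sphere, which links to that side's thigh-twist capsule with free rotation.
# The belt's Dyn capsule hangs from its own base capsule with uneven swing limits and no constraint to the thighs.
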PhysicsAsset /BRCosmetics/Characters/Player/Male/Medium/Bodies/M_MED_GlitchPrince/Meshes/M_MED_GlitchPrince_pants_Physics
bodies {
 Base_belt [{shapeType=Capsule, length=0, radius=2, shapeName=Base_belt_capsule}]
 Dyn_belt [{shapeType=Capsule, length=3, radius=5, shapeName=Dyn_belt_capsule, center=<14.4736 0 0>, rotation=<rot 0 0.0002 88.3988>}]
 Dyn_pants_l_01 [{shapeType=Capsule, length=3, radius=5, shapeName=Dyn_pants_l_01_capsule, center=<11.8719 0 0>}]
 Dyn_pants_l_02 [{shapeType=Capsule, length=3, radius=5, shapeName=Dyn_pants_l_02_capsule, center=<11.8719 0 0>}]
 Dyn_pants_l_03 [{shapeType=Capsule, length=3, radius=5, shapeName=Dyn_pants_l_03_capsule, center=<11.8719 0 0>}]
 Dyn_pants_l_04 [{shapeType=Capsule, length=3, radius=5, shapeName=Dyn_pants_l_04_capsule, center=<11.8719 0 0>}]
 Dyn_pants_r_01 [{shapeType=Capsule, length=3, radius=5, shapeName=Dyn_pants_r_01_capsule, center=<11.8719 0 0>}]
 Dyn_pants_r_02 [{shapeType=Capsule, length=3, radius=5, shapeName=Dyn_pants_r_02_capsule, center=<11.8719 0 0>}]
 Dyn_pants_r_03 [{shapeType=Capsule, length=3, radius=5, shapeName=Dyn_pants_r_03_capsule, center=<11.8719 0 0>}]
 Dyn_pants_r_04 [{shapeType=Capsule, length=3, radius=5, shapeName=Dyn_pants_r_04_capsule, center=<11.8719 0 0>}]
 base_pants_l_01 [{shapeType=Sphere, radius=0.505, shapeName=base_pants_l_01_sphere}]
 base_pants_l_02 [{shapeType=Sphere, radius=0.505, shapeName=base_pants_l_02_sphere}]
 base_pants_l_03 [{shapeType=Sphere, radius=0.505, shapeName=base_pants_l_03_sphere}]
 base_pants_l_04 [{shapeType=Sphere, radius=0.505, shapeName=base_pants_l_04_sphere}]
 base_pants_r_01 [{shapeType=Sphere, radius=0.505, shapeName=base_pants_r_01_sphere}]
 base_pants_r_02 [{shapeType=Sphere, radius=0.505, shapeName=base_pants_r_02_sphere}]
 base_pants_r_03 [{shapeType=Sphere, radius=0.505, shapeName=base_pants_r_03_sphere}]
 base_pants_r_04 [{shapeType=Sphere, radius=0.505, shapeName=base_pants_r_04_sphere}]
 thigh_twist_01_l [{shapeType=Capsule, length=32.5, radius=6, shapeName=thigh_twist_01_l_capsule, center=<-16.008 0.0001 0>, rotation=<rot 90 -0.0004 0>}]
 thigh_twist_01_r [{shapeType=Capsule, length=32.5, radius=6, shapeName=thigh_twist_01_r_capsule, center=<15.2437 -0.0001 0>, rotation=<rot 90 -0.0004 0>}]
}
constraints [
 {bone1Name=Dyn_pants_r_01, bone2Name=base_pants_r_01, swing1LimitDegrees=10, swing2LimitDegrees=10, twistLimitDegrees=1}
 {bone1Name=Dyn_pants_r_02, bone2Name=base_pants_r_02, swing1LimitDegrees=10, swing2LimitDegrees=10, twistLimitDegrees=1}
 {bone1Name=Dyn_pants_r_03, bone2Name=base_pants_r_03, swing1LimitDegrees=10, swing2LimitDegrees=10, twistLimitDegrees=1}
 {bone1Name=Dyn_pants_r_04, bone2Name=base_pants_r_04, swing1LimitDegrees=10, swing2LimitDegrees=10, twistLimitDegrees=1}
 {bone1Name=Dyn_pants_l_01, bone2Name=base_pants_l_01, swing1LimitDegrees=10, swing2LimitDegrees=10, twistLimitDegrees=1}
 {bone1Name=Dyn_pants_l_02, bone2Name=base_pants_l_02, swing1LimitDegrees=10, swing2LimitDegrees=10, twistLimitDegrees=1}
 {bone1Name=Dyn_pants_l_03, bone2Name=base_pants_l_03, swing1LimitDegrees=10, swing2LimitDegrees=10, twistLimitDegrees=1}
 {bone1Name=Dyn_pants_l_04, bone2Name=base_pants_l_04, swing1LimitDegrees=10, swing2LimitDegrees=10, twistLimitDegrees=1}
 {bone1Name=Dyn_belt, bone2Name=Base_belt, swing1LimitDegrees=5, swing2LimitDegrees=20, twistLimitDegrees=1}
 {
  bone1Name=base_pants_r_04
  bone2Name=thigh_twist_01_r
  swing1LimitDegrees=45
  swing1Motion=Free
  swing2LimitDegrees=45
  swing2Motion=Free
  twistLimitDegrees=45
  twistMotion=Free
 }
 {
  bone1Name=base_pants_r_03
  bone2Name=thigh_twist_01_r
  swing1LimitDegrees=45
  swing1Motion=Free
  swing2LimitDegrees=45
  swing2Motion=Free
  twistLimitDegrees=45
  twistMotion=Free
 }
 {
  bone1Name=base_pants_r_02
  bone2Name=thigh_twist_01_r
  swing1LimitDegrees=45
  swing1Motion=Free
  swing2LimitDegrees=45
  swing2Motion=Free
  twistLimitDegrees=45
  twistMotion=Free
 }
 {
  bone1Name=base_pants_r_01
  bone2Name=thigh_twist_01_r
  swing1LimitDegrees=45
  swing1Motion=Free
  swing2LimitDegrees=45
  swing2Motion=Free
  twistLimitDegrees=45
  twistMotion=Free
 }
 {
  bone1Name=base_pants_l_04
  bone2Name=thigh_twist_01_l
  swing1LimitDegrees=45
  swing1Motion=Free
  swing2LimitDegrees=45
  swing2Motion=Free
  twistLimitDegrees=45
  twistMotion=Free
 }
 {
  bone1Name=base_pants_l_03
  bone2Name=thigh_twist_01_l
  swing1LimitDegrees=45
  swing1Motion=Free
  swing2LimitDegrees=45
  swing2Motion=Free
  twistLimitDegrees=45
  twistMotion=Free
 }
 {
  bone1Name=base_pants_l_02
  bone2Name=thigh_twist_01_l
  swing1LimitDegrees=45
  swing1Motion=Free
  swing2LimitDegrees=45
  swing2Motion=Free
  twistLimitDegrees=45
  twistMotion=Free
 }
 {
  bone1Name=base_pants_l_01
  bone2Name=thigh_twist_01_l
  swing1LimitDegrees=45
  swing1Motion=Free
  swing2LimitDegrees=45
  swing2Motion=Free
  twistLimitDegrees=45
  twistMotion=Free
 }
]
