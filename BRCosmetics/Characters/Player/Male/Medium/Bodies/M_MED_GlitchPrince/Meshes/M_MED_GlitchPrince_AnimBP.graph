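# Post-processes the GlitchPrince character body's copied parent-mesh pose to drive secondary motion on collar, jacket and pants.
# Weighted collar constraints and a Control Rig feed pants correctives, jacket look-ats, then two pelvis-space RigidBody simulations and jacket adjustments.
# The event graph derives corrective alphas from head, calf and thigh rotations and crouch/skydive states, and updates gravity through a deprecated override function.
# Anim Dynamics nodes sit disconnected from the output.
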AnimBlueprint /BRCosmetics/Characters/Player/Male/Medium/Bodies/M_MED_GlitchPrince/Meshes/M_MED_GlitchPrince_AnimBP
target_skeleton=&/M_MED_GlitchPrince_Skeleton
graphs [
 AnimGraph {nodes [
   Anim Dynamics - Bone: Dyn_pants_l_01!(ComponentPose, Alpha=1, GravityOverride:Vector="0,0,-980", ExternalForce="0,0,-980") -> (Pose) {
    angularDampingOverride=0.9999
    angularSpringConstant=100
    bAngularSpring=true
    bOverrideAngularDamping=true
    bUseGravityOverride=true
    simulationSpace=BoneRelative
    alphaBoolBlend {blendOption=Linear}
    boundBone {boneName=Dyn_pants_l_01}
    componentAppliedLinearAccClamp=<1500 1500 1500>
    componentLinearAccScale=<0.6 0.6 0.8>
    componentLinearVelScale=<0.6 0.6 0.8>
    physicsBodyDefinitions [{
      boundBone {boneName=Dyn_pants_l_01}
      boxExtents=<10 10 10>
      constraintSetup {angularLimitsMax=<0 20 0>, angularLimitsMin=<0 -20 -5>, angularTarget=<0 0 0>}
      localJointOffset=<0.1 0 0>
     }]
    relativeSpaceBone {boneName=thigh_l}
   }
   n1 = Anim Dynamics - Bone: dyn_collar_back_r_01!(ComponentPose, Alpha=1, GravityOverride:Vector="0,0,-980", ExternalForce="0,0,-980") -> (Pose) {
    angularDampingOverride=0.9999
    angularSpringConstant=100
    bAngularSpring=true
    bOverrideAngularDamping=true
    bUseGravityOverride=true
    simulationSpace=BoneRelative
    alphaBoolBlend {blendOption=Linear}
    boundBone {boneName=dyn_collar_back_r_01}
    componentAppliedLinearAccClamp=<1500 1500 1500>
    componentLinearAccScale=<0.6 0.6 0.8>
    componentLinearVelScale=<0.6 0.6 0.8>
    physicsBodyDefinitions [{
      boundBone {boneName=dyn_collar_back_r_01}
      boxExtents=<10 10 10>
      constraintSetup {angularLimitsMax=<2.5 0 2.5>, angularLimitsMin=<-2.5 -10 -2.5>, angularTarget=<0 0 0>}
      localJointOffset=<1 0 1>
     }]
    relativeSpaceBone {boneName=pelvis}
   }
   Anim Dynamics - Bone: dyn_collar_back_l_01!(ComponentPose=@n1, Alpha=1, GravityOverride:Vector="0,0,-980", ExternalForce="0,0,-980") -> (Pose) {
    angularDampingOverride=0.9999
    angularSpringConstant=100
    bAngularSpring=true
    bOverrideAngularDamping=true
    bUseGravityOverride=true
    simulationSpace=BoneRelative
    alphaBoolBlend {blendOption=Linear}
    boundBone {boneName=dyn_collar_back_l_01}
    componentAppliedLinearAccClamp=<1500 1500 1500>
    componentLinearAccScale=<0.6 0.6 0.8>
    componentLinearVelScale=<0.6 0.6 0.8>
    physicsBodyDefinitions [{
      boundBone {boneName=dyn_collar_back_l_01}
      boxExtents=<10 10 10>
      constraintSetup {angularLimitsMax=<2.5 0 2.5>, angularLimitsMin=<-2.5 -10 -2.5>, angularTarget=<0 0 0>}
      localJointOffset=<1 0 1>
     }]
    relativeSpaceBone {boneName=pelvis}
   }
   Anim Dynamics - Bone: dyn_collar_front_l!(ComponentPose, Alpha=1, GravityOverride:Vector="0,0,-980", ExternalForce="0,0,-980") -> (Pose) {
    angularDampingOverride=0.9999
    angularSpringConstant=100
    bAngularSpring=true
    bOverrideAngularDamping=true
    bUseGravityOverride=true
    simulationSpace=BoneRelative
    alphaBoolBlend {blendOption=Linear}
    boundBone {boneName=dyn_collar_front_l}
    componentAppliedLinearAccClamp=<1500 1500 1500>
    componentLinearAccScale=<0.6 0.6 0.8>
    componentLinearVelScale=<0.6 0.6 0.8>
    physicsBodyDefinitions [{
      boundBone {boneName=dyn_collar_front_l}
      boxExtents=<10 10 10>
      constraintSetup {angularLimitsMax=<5 0 5>, angularLimitsMin=<-5 -10 -5>, angularTarget=<0 0 0>}
      localJointOffset=<1 0 -1>
     }]
    relativeSpaceBone {boneName=pelvis}
   }
   "Constraint - dyn_collar_r_mstr (PPPPPPPPPPP)"!(
    ComponentPose
    Alpha=1
    ConstraintWeights_0=0
    ConstraintWeights_1:float=0.0065
    ConstraintWeights_2:float=0.0225
    ConstraintWeights_3:float=0.3699
    ConstraintWeights_4:float=0.0035
    ConstraintWeights_5:float=0.0984
    ConstraintWeights_6:float=0.0748
    ConstraintWeights_7:float=0.103
    ConstraintWeights_8:float=0.316
    ConstraintWeights_9:float=0.0052
    ConstraintWeights_10:float=0.0001
   ) -> (Pose) {
    alphaBoolBlend {blendOption=Linear}
    boneToModify {boneName=dyn_collar_r_mstr}
    constraintSetup [
     {targetBone {boneName=spine_02}}
     {targetBone {boneName=spine_03}}
     {targetBone {boneName=spine_04}}
     {targetBone {boneName=spine_05}}
     {targetBone {boneName=clavicle_l}}
     {targetBone {boneName=clavicle_r}}
     {targetBone {boneName=neck_01}}
     {targetBone {boneName=neck_02}}
     {targetBone {boneName=head}}
     {targetBone {boneName=dyn_collar_front_r}}
     {targetBone {boneName=dyn_collar_front_l}}
    ]
    constraintWeights [1,1,1,1,1,1,1,1,1,1,1]
   }
   n2 = "Transform (Modify) Bone - Bone: dyn_jacket_fr_l_mstr"(Alpha=1, Translation="0.123,-0.941,-0.328", Rotation="32.135,-4.173,49.923", Scale="1,1,1") {
    rotationMode=BMM_Additive
    scaleMode=BMM_Additive
    scaleSpace=BCS_BoneSpace
    translationMode=BMM_Additive
    translationSpace=BCS_BoneSpace
    alphaBoolBlend {blendOption=Linear}
    alphaScaleBiasClamp {bInterpResult=true, interpSpeedDecreasing=9, interpSpeedIncreasing=30}
    boneToModify {boneName=dyn_jacket_fr_l_mstr}
    rotation=<rot 32.135 -4.173 49.923>
    translation=<0.123 -0.941 -0.328>
   }
   "Transform (Modify) Bone - Bone: dyn_jacket_s_l_mstr"(ComponentPose=@n2, Alpha=1, Translation="-1.688,-1.330,-1.631", Rotation="-17.920,-3.749,31.829", Scale="0.7,1,1") {
    rotationMode=BMM_Additive
    scaleMode=BMM_Additive
    scaleSpace=BCS_BoneSpace
    translationMode=BMM_Additive
    translationSpace=BCS_BoneSpace
    alphaBoolBlend {blendOption=Linear}
    alphaScaleBiasClamp {bInterpResult=true, interpSpeedDecreasing=7, interpSpeedIncreasing=7}
    boneToModify {boneName=dyn_jacket_s_l_mstr}
    rotation=<rot -17.92 -3.749 31.829>
    scale {x=0.7}
    translation=<-1.688 -1.33 -1.631>
   }
   "Transform (Modify) Bone - Bone: dyn_jacket_l_bk_master"(Alpha=1, Translation="0,0,0", Rotation="0,0,30", Scale="1,1,1") {
    rotationMode=BMM_Additive
    rotationSpace=BCS_BoneSpace
    translationMode=BMM_Additive
    translationSpace=BCS_BoneSpace
    alphaBoolBlend {blendOption=Linear}
    boneToModify {boneName=dyn_jacket_l_bk_master}
    rotation {roll=30}
   }
   n3 = "Constraint - dyn_collar_l_mstr (PPPPPP)"!(
    ComponentPose=Local To Component(LocalPose=Copy Pose From Mesh() {bCopyCurves=true, bCopyCustomAttributes=true, bUseAttachedParent=true, bUseMeshPose=true})
    Alpha=1
    ConstraintWeights_0=0.0004
    ConstraintWeights_1:float=0.0112
    ConstraintWeights_2:float=0.5941
    ConstraintWeights_3:float=0.3228
    ConstraintWeights_4:float=0.0191
    ConstraintWeights_5:float=0.0523
   ) -> (Pose) {
    alphaBoolBlend {blendOption=Linear}
    boneToModify {boneName=dyn_collar_l_mstr}
    constraintSetup [
     {targetBone {boneName=spine_03}}
     {targetBone {boneName=spine_04}}
     {targetBone {boneName=spine_05}}
     {targetBone {boneName=clavicle_l}}
     {targetBone {boneName=neck_01}}
     {targetBone {boneName=neck_02}}
    ]
    constraintWeights [1,1,1,1,1,1]
   }
   n4 = "Constraint - dyn_collar_fr_mstr (PPPPPPPP)"!(
    ComponentPose=@n3
    Alpha=1
    ConstraintWeights_0=0.0007
    ConstraintWeights_1:float=0.0235
    ConstraintWeights_2:float=0.0829
    ConstraintWeights_3:float=0.674
    ConstraintWeights_4:float=0.1166
    ConstraintWeights_5:float=0.0295
    ConstraintWeights_6:float=0.0719
    ConstraintWeights_7:float=0.001
   ) -> (Pose) {
    alphaBoolBlend {blendOption=Linear}
    boneToModify {boneName=dyn_collar_fr_mstr}
    constraintSetup [
     {targetBone {boneName=spine_02}}
     {targetBone {boneName=spine_03}}
     {targetBone {boneName=spine_04}}
     {targetBone {boneName=spine_05}}
     {targetBone {boneName=clavicle_l}}
     {targetBone {boneName=neck_01}}
     {targetBone {boneName=neck_02}}
     {targetBone {boneName=head}}
    ]
    constraintWeights [1,1,1,1,1,1,1,1]
   }
   n5 = "Constraint - dyn_collar_bk_mstr (PPPPPPPPPP)"!(
    ComponentPose=@n4
    Alpha=1
    ConstraintWeights_0=0.0175
    ConstraintWeights_1:float=0.0054
    ConstraintWeights_2:float=0.5952
    ConstraintWeights_3:float=0.0311
    ConstraintWeights_4:float=0.0158
    ConstraintWeights_5:float=0.0881
    ConstraintWeights_6:float=0.2006
    ConstraintWeights_7:float=0.0461
    ConstraintWeights_8:float=0.0001
    ConstraintWeights_9:float=0.0001
   ) -> (Pose) {
    alphaBoolBlend {blendOption=Linear}
    boneToModify {boneName=dyn_collar_bk_mstr}
    constraintSetup [
     {targetBone {boneName=spine_03}}
     {targetBone {boneName=spine_04}}
     {targetBone {boneName=spine_05}}
     {targetBone {boneName=clavicle_l}}
     {targetBone {boneName=clavicle_r}}
     {targetBone {boneName=neck_01}}
     {targetBone {boneName=neck_02}}
     {targetBone {boneName=head}}
     {targetBone {boneName=dyn_collar_l}}
     {targetBone {boneName=dyn_collar_r}}
    ]
    constraintWeights [1,1,1,1,1,1,1,1,1,1]
   }
   n6 = "Transform (Modify) Bone - Bone: dyn_collar_bk_mstr"(ComponentPose=@n5, Alpha=Get Head_rot() -> (Head_rot:double), Translation="0,0,0", Rotation="-78.924,179.866,-179.821", Scale="1,1,1") {
    rotationMode=BMM_Additive
    rotationSpace=BCS_BoneSpace
    translationMode=BMM_Additive
    translationSpace=BCS_BoneSpace
    alphaBoolBlend {blendOption=Linear}
    boneToModify {boneName=dyn_collar_bk_mstr}
    rotation=<rot -78.924 179.866 -179.821>
   }
   n7 = "Transform (Modify) Bone - Bone: dyn_collar_lbk_mstr"(ComponentPose=@n6, Alpha=Get Head_rot(), Translation="0,0,0", Rotation="0,0,0", Scale="1,1,1") {
    rotationMode=BMM_Additive
    rotationSpace=BCS_BoneSpace
    translationMode=BMM_Additive
    translationSpace=BCS_BoneSpace
    alphaBoolBlend {blendOption=Linear}
    boneToModify {boneName=dyn_collar_lbk_mstr}
   }
   n8 = "Transform (Modify) Bone - Bone: dyn_collar_rbk_mstr"(ComponentPose=@n7, Alpha=Get Head_rot(), Translation="0,0,0", Rotation="0,0,0", Scale="1,1,1") {
    rotationMode=BMM_Additive
    rotationSpace=BCS_BoneSpace
    translationMode=BMM_Additive
    translationSpace=BCS_BoneSpace
    alphaBoolBlend {blendOption=Linear}
    boneToModify {boneName=dyn_collar_rbk_mstr}
   }
   n9 = Control Rig(Source=Component To Local(ComponentPose=@n8)) {controlRigAssetReference {blueprintRigClass=&/M_MED_GlitchPrince_CtrlRig.M_MED_GlitchPrince_CtrlRig_C}}
   n10 = "Transform (Modify) Bone - Bone: base_pants_l_02"(
    ComponentPose=Local To Component(LocalPose=@n9)
    Alpha=Get Calf_l_rot() -> (Calf_l_rot:double)
    Translation="-6.778,18.172,1.086"
    Rotation="-23.166,-147.579,22.882"
    Scale="1,1,1"
   ) {
    rotationMode=BMM_Additive
    rotationSpace=BCS_BoneSpace
    translationMode=BMM_Additive
    translationSpace=BCS_BoneSpace
    alphaBoolBlend {blendOption=Linear}
    boneToModify {boneName=base_pants_l_02}
    rotation=<rot -23.166 -147.579 22.882>
    translation=<-6.778 18.172 1.086>
   }
   n11 = "Transform (Modify) Bone - Bone: base_pants_r_01"(
    ComponentPose=@n10
    Alpha=Get Calf_r_rot() -> (Calf_r_rot:double)
    Translation="-1.012,19.750,2.822"
    Rotation="-0.169,-128.716,0.070"
    Scale="1,1,1"
   ) {
    rotationMode=BMM_Additive
    rotationSpace=BCS_BoneSpace
    translationMode=BMM_Additive
    translationSpace=BCS_BoneSpace
    alphaBoolBlend {blendOption=Linear}
    boneToModify {boneName=base_pants_r_01}
    rotation=<rot -0.169 -128.716 0.07>
    translation=<-1.012 19.75 2.822>
   }
   n12 = "Transform (Modify) Bone - Bone: base_pants_l_03"(ComponentPose=@n11, Alpha=Get Calf_l_rot(), Translation="-1.157,-1.068,4.922", Rotation="-72.027,111.421,-96.631", Scale="1,1,1") {
    rotationMode=BMM_Additive
    rotationSpace=BCS_BoneSpace
    translationMode=BMM_Additive
    translationSpace=BCS_BoneSpace
    alphaBoolBlend {blendOption=Linear}
    boneToModify {boneName=base_pants_l_03}
    rotation=<rot -72.027 111.421 -96.631>
    translation=<-1.157 -1.068 4.922>
   }
   n13 = "Transform (Modify) Bone - Bone: base_pants_r_02"(ComponentPose=@n12, Alpha=Get Calf_r_rot(), Translation="-2.799,-1.202,-3.955", Rotation="58.211,22.369,18.202", Scale="1,1,1") {
    rotationMode=BMM_Additive
    rotationSpace=BCS_BoneSpace
    translationMode=BMM_Additive
    translationSpace=BCS_BoneSpace
    alphaBoolBlend {blendOption=Linear}
    boneToModify {boneName=base_pants_r_02}
    rotation=<rot 58.211 22.369 18.202>
    translation=<-2.799 -1.202 -3.955>
   }
   n14 = "Transform (Modify) Bone - Bone: base_pants_l_04"(ComponentPose=@n13, Alpha=Get Calf_l_rot(), Translation="-0.380,-0.024,-4.081", Rotation="76.035,-179.905,-179.957", Scale="1,1,1") {
    rotationMode=BMM_Additive
    rotationSpace=BCS_BoneSpace
    translationMode=BMM_Additive
    translationSpace=BCS_BoneSpace
    alphaBoolBlend {blendOption=Linear}
    boneToModify {boneName=base_pants_l_04}
    rotation=<rot 76.035 -179.905 -179.957>
    translation=<-0.38 -0.024 -4.081>
   }
   n15 = "Transform (Modify) Bone - Bone: base_pants_r_03"(ComponentPose=@n14, Alpha=Get Calf_r_rot(), Translation="-1.271,0.433,4.428", Rotation="-80.961,0.545,-0.596", Scale="1,1,1") {
    rotationMode=BMM_Additive
    rotationSpace=BCS_BoneSpace
    translationMode=BMM_Additive
    translationSpace=BCS_BoneSpace
    alphaBoolBlend {blendOption=Linear}
    boneToModify {boneName=base_pants_r_03}
    rotation=<rot -80.961 0.545 -0.596>
    translation=<-1.271 0.433 4.428>
   }
   n16 = "Transform (Modify) Bone - Bone: pelvis_target_back"(ComponentPose=@n15, Alpha=Get back_orient_out() -> (back_orient_out:double), Translation="0,-10,0", Rotation="0,0,0", Scale="1,1,1") {
    rotationMode=BMM_Additive
    rotationSpace=BCS_BoneSpace
    translationMode=BMM_Additive
    translationSpace=BCS_BoneSpace
    alphaBoolBlend {blendOption=Linear}
    boneToModify {boneName=pelvis_target_back}
    translation {y=-10}
   }
   n17 = "Transform (Modify) Bone - Bone: pelvis_target_back"(ComponentPose=@n16, Alpha=Get back_in() -> (back_in:double), Translation="0,2,0", Rotation="0,0,0", Scale="1,1,1") {
    rotationMode=BMM_Additive
    rotationSpace=BCS_BoneSpace
    translationMode=BMM_Additive
    translationSpace=BCS_BoneSpace
    alphaBoolBlend {blendOption=Linear}
    boneToModify {boneName=pelvis_target_back}
    translation {y=2}
   }
   n18 = Look At - Bone: dyn_jacket_l_fr_orient(ComponentPose=@n17, Alpha=1) {
    alphaBoolBlend {blendOption=Linear}
    boneToModify {boneName=dyn_jacket_l_fr_orient}
    lookAtLocation {x=0}
    lookAtTarget {boneReference {boneName=pelvis_target_fr_l}}
    lookAt_Axis {axis=<1 0>}
    lookUp_Axis {axis {x=0, z=1}}
   }
   n19 = Look At - Bone: dyn_jacket_r_fr_orient(ComponentPose=@n18, Alpha=1) {
    alphaBoolBlend {blendOption=Linear}
    boneToModify {boneName=dyn_jacket_r_fr_orient}
    lookAtLocation {x=0}
    lookAtTarget {boneReference {boneName=pelvis_target_fr_r}}
    lookAt_Axis {axis=<1 0>}
    lookUp_Axis {axis {x=0, z=1}}
   }
   n20 = Look At - Bone: dyn_jacket_l_bk_orient(ComponentPose=@n19, Alpha=1) {
    alphaBoolBlend {blendOption=Linear}
    boneToModify {boneName=dyn_jacket_l_bk_orient}
    lookAtLocation {x=0}
    lookAtTarget {boneReference {boneName=pelvis_target_back}}
    lookAt_Axis {axis=<1 0>}
    lookUp_Axis {axis {x=0, z=1}}
   }
   n21 = Look At - Bone: dyn_jacket_r_bk_orient(ComponentPose=@n20, Alpha=1) {
    alphaBoolBlend {blendOption=Linear}
    boneToModify {boneName=dyn_jacket_r_bk_orient}
    lookAtLocation {x=0}
    lookAtTarget {boneReference {boneName=pelvis_target_back}}
    lookAt_Axis {axis=<1 0>}
    lookUp_Axis {axis {x=0, z=1}}
   }
   n22 = "Transform (Modify) Bone - Bone: dyn_jacket_r_bk_master"(ComponentPose=@n21, Alpha=1, Translation="0,0,0", Rotation="0,0,-30", Scale="1,1,1") {
    rotationMode=BMM_Additive
    rotationSpace=BCS_BoneSpace
    translationMode=BMM_Additive
    translationSpace=BCS_BoneSpace
    alphaBoolBlend {blendOption=Linear}
    boneToModify {boneName=dyn_jacket_r_bk_master}
    rotation {roll=-30}
   }
   n23 = "Transform (Modify) Bone - Bone: dyn_jacket_l_bk_01"(ComponentPose=@n22, Alpha=1, Translation="0,0,0", Rotation="0,0,15", Scale="1,1,1") {
    rotationMode=BMM_Additive
    rotationSpace=BCS_BoneSpace
    translationMode=BMM_Additive
    translationSpace=BCS_BoneSpace
    alphaBoolBlend {blendOption=Linear}
    boneToModify {boneName=dyn_jacket_l_bk_01}
    rotation {roll=15}
   }
   n24 = "Transform (Modify) Bone - Bone: dyn_jacket_l_01"(ComponentPose=@n23, Alpha=1, Translation="0,0,0", Rotation="-15,0,0", Scale="1,1,1") {
    rotationMode=BMM_Additive
    rotationSpace=BCS_BoneSpace
    translationMode=BMM_Additive
    translationSpace=BCS_BoneSpace
    alphaBoolBlend {blendOption=Linear}
    boneToModify {boneName=dyn_jacket_l_01}
    rotation {pitch=-15}
   }
   n25 = "Transform (Modify) Bone - Bone: dyn_jacket_l_fr_master"(ComponentPose=@n24, Alpha=Get thigh_rot_l(), Translation="-2.428,0.270,4.837", Rotation="0,0,30", Scale="1,1,1") {
    rotationMode=BMM_Additive
    scaleMode=BMM_Additive
    scaleSpace=BCS_BoneSpace
    translationMode=BMM_Additive
    translationSpace=BCS_BoneSpace
    alphaBoolBlend {blendOption=Linear}
    alphaScaleBiasClamp {bInterpResult=true, interpSpeedDecreasing=9, interpSpeedIncreasing=30}
    boneToModify {boneName=dyn_jacket_l_fr_master}
    rotation {roll=30}
    translation=<-2.428 0.27 4.837>
   }
   n26 = Get GravityOverride()
   n27 = RigidBody!(ComponentPose=@n25, Alpha, OverrideWorldGravity:Vector=@n26, ExternalForce=@n26) -> (Pose) {
    bClampLinearTranslationLimitToRefPose=true
    bOverrideWorldGravity=true
    overridePhysicsAsset=&/M_MED_GlitchPrince_Physics
    simulationSpace=BaseBoneSpace
    alphaBoolBlend {blendOption=Linear}
    baseBoneRef {boneName=pelvis}
    componentAppliedLinearAccClamp=<1500 1500 1500>
    componentLinearAccScale=<0.35 0.35 0.45>
    overrideWorldGravity {z=-980}
    simSpaceSettings {
     maxAngularAcceleration=100
     maxAngularVelocity=3
     maxLinearAcceleration=700
     maxLinearVelocity=500
     velocityScaleZ=0
     worldAlpha=1
     externalLinearDragV=<0.4 0.4 0.4>
    }
   }
   n28 = Get GravityOverride()
   n29 = RigidBody(ComponentPose=@n27, OverrideWorldGravity=@n28, ExternalForce=@n28) {
    bClampLinearTranslationLimitToRefPose=true
    bOverrideWorldGravity=true
    overridePhysicsAsset=&/M_MED_GlitchPrince_pants_Physics
    simulationSpace=BaseBoneSpace
    alphaBoolBlend {blendOption=Linear}
    baseBoneRef {boneName=pelvis}
    componentAppliedLinearAccClamp=<1500 1500 1500>
    componentLinearAccScale=<0.35 0.35 0.45>
    overrideWorldGravity {z=-980}
    simSpaceSettings {
     maxAngularAcceleration=100
     maxAngularVelocity=3
     maxLinearAcceleration=700
     maxLinearVelocity=500
     velocityScaleZ=0
     worldAlpha=1
     externalLinearDragV=<0.4 0.4 0.4>
    }
   }
   n30 = "Transform (Modify) Bone - Bone: dyn_jacket_l_master"!(ComponentPose=@n29, bAlphaBoolEnabled:bool=True, Translation="0,0,0", Rotation="10,0,0", Scale="1,1,1") -> (Pose) {
    alphaInputType=Bool
    rotationMode=BMM_Additive
    rotationSpace=BCS_BoneSpace
    translationMode=BMM_Additive
    translationSpace=BCS_BoneSpace
    alphaBoolBlend {blendOption=Linear}
    boneToModify {boneName=dyn_jacket_l_master}
    rotation {pitch=10}
   }
   n31 = "Transform (Modify) Bone - Bone: dyn_jacket_l_bk_master"!(ComponentPose=@n30, bAlphaBoolEnabled:bool=True, Translation="0,0,0", Rotation="10,0,0", Scale="1,1,1") -> (Pose) {
    alphaInputType=Bool
    rotationMode=BMM_Additive
    rotationSpace=BCS_BoneSpace
    translationMode=BMM_Additive
    translationSpace=BCS_BoneSpace
    alphaBoolBlend {blendOption=Linear}
    boneToModify {boneName=dyn_jacket_l_bk_master}
    rotation {pitch=10}
   }
   n32 = "Transform (Modify) Bone - Bone: dyn_jacket_l_fr_master"!(ComponentPose=@n31, bAlphaBoolEnabled:bool=True, Translation="0,0,0", Rotation="10,0,0", Scale="1,1,1") -> (Pose) {
    alphaInputType=Bool
    rotationMode=BMM_Additive
    rotationSpace=BCS_BoneSpace
    translationMode=BMM_Additive
    translationSpace=BCS_BoneSpace
    alphaBoolBlend {blendOption=Linear}
    boneToModify {boneName=dyn_jacket_l_fr_master}
    rotation {pitch=10}
   }
   n33 = Get back_constrain() -> (back_constrain:double)
   n34 = "Constraint - dyn_jacket_l_bk_02 (P)"(ComponentPose=@n32, Alpha=@n33, ConstraintWeights_0=1) {
    alphaBoolBlend {blendOption=Linear}
    boneToModify {boneName=dyn_jacket_l_bk_02}
    constraintSetup [{targetBone {boneName=pelvis}}]
    constraintWeights [1]
   }
   n35 = "Constraint - dyn_jacket_r_bk_02 (P)"(ComponentPose=@n34, Alpha=@n33, ConstraintWeights_0=1) {
    alphaBoolBlend {blendOption=Linear}
    boneToModify {boneName=dyn_jacket_r_bk_02}
    constraintSetup [{targetBone {boneName=pelvis}}]
    constraintWeights [1]
   }
   Output Pose(Result=Component To Local(ComponentPose=@n35)) {layerGroup=None}
  ]}
 EventGraph {nodes [
   Event BlueprintUpdateAnimation()
   > n1 = Sequence()
   > n2 = Gravity Override State Function DEPRECATED(
    CharacterPartAnimInstance=Self-Reference()
    GameTimeInSeconds=GetGameTimeInSeconds()
    GravityOverrideParams=Get GravityOverrideParams() -> (GravityOverrideParams:Gravity Override Params Struct Structure)
   )
   > Set GravityOverride(GravityOverride=@n2)
   n3 = Get bIsSkydiveDiveMode()
   Set back_orient_out(
    execute=@n1.then_1
    back_orient_out:double=FInterpTo(Current=Get back_orient_out(), Target=SelectFloat(A=1, B=0, bPickA=@n3), DeltaTime=0.5, InterpSpeed=0.6)
   ) -> (Output_Get:double)
   n4 = AND Boolean(A=Get bIsSkydiveFloating(), B=Get bRigidBodyEnabled())
   > Set back_constrain(back_constrain:double=FInterpTo(Current=Get back_constrain(), Target=SelectFloat(A=0.25, B=0, bPickA=@n4), DeltaTime=0.5, InterpSpeed=0.6)) -> (Output_Get:double)
   n5 = OR Boolean(A=Get bIsCrouching(), B=NOT Boolean(A=Get bRigidBodyEnabled()))
   > Set crouch_alpha(crouch_alpha:double=FInterpTo(Current=Get crouch_alpha() -> (crouch_alpha:double), Target=SelectFloat(A=1, B=0, bPickA=@n5), DeltaTime=0.5, InterpSpeed=0.4)) -> (Output_Get:double)
   > Set emote alpha(emote alpha:double=SelectFloat(A=0, B=1, bPickA=Get bIsPlayingEmote())) -> (Output_Get:double)
   n6 = AND Boolean(A=NOT Boolean(A=Get bRigidBodyEnabled()), B=Get bIsCrouching())
   > Set back_in(back_in:double=FInterpTo(Current=Get back_in(), Target=SelectFloat(A=1, B=0, bPickA=@n6), DeltaTime=0.5, InterpSpeed=0.4)) -> (Output_Get:double)
   n7 = GetOwnerSkeletalMeshComponent()
   > Set Calf_l_rot(Calf_l_rot:double=MapRangeClamped(
     Value=GetDeltaTransformFromRefPose!(self=@n7, BoneName=calf_l, BaseName=thigh_l) -> (
      ReturnValue_Location:Vector
      ReturnValue_Rotation_Roll:float
      ReturnValue_Rotation_Pitch:float
      ReturnValue_Rotation_Yaw:float
      ReturnValue_Scale:Vector
     ).ReturnValue_Rotation_Yaw
     InRangeA=0
     InRangeB=150
    )) -> (Output_Get:double)
   n8 = GetOwnerSkeletalMeshComponent()
   > Set Calf_r_rot(Calf_r_rot:double=MapRangeClamped(Value=GetDeltaTransformFromRefPose(self=@n8, BoneName=calf_r, BaseName=thigh_r).ReturnValue_Rotation_Yaw, InRangeA=0, InRangeB=150)) -> (Output_Get:double)
   n9 = GetOwnerSkeletalMeshComponent()
   > Set Head_rot(Head_rot:double=MapRangeClamped(Value=GetDeltaTransformFromRefPose(self=@n9, BoneName=head, BaseName=spine_05).ReturnValue_Rotation_Yaw, InRangeA=0, InRangeB=-100)) -> (Output_Get:double)
   > n10 = Branch(Condition=Get bIsCrouching())
   n11 = GetOwnerSkeletalMeshComponent()
   > Set thigh_rot_l(thigh_rot_l=MapRangeClamped(
     Value=GetDeltaTransformFromRefPose(self=@n11, BoneName=thigh_l, BaseName=pelvis).ReturnValue_Rotation_Yaw
     InRangeA=-100
     InRangeB=-20
     OutRangeA=1
     OutRangeB=0
    ))
   Set thigh_rot_l(execute=@n10.else, thigh_rot_l=0)
   OR Boolean(A=Get bIsSkydiveFloating(), B=@n3)
  ]}
]
variables [
 back_constrain {type=number}
 back_in {type=number}
 back_orient_out {type=number}
 Calf_l_rot {type=number}
 Calf_r_rot {type=number}
 crouch_alpha {type=number}
 emote alpha {type=number}
 GravityOverride {type=Vector}
 GravityOverrideParams {type=GravityOverrideParamsStruct}
 Head_rot {type=number}
 thigh_rot_l {type=number}
]
