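# Physics asset for the face accessory mesh of the GlitchPrince male medium player cosmetic, covering its crown and hair bones.
# A small head capsule anchors sphere bodies for the crown and four hair bones, each constrained directly to the head.
# The crown's swing and twist are locked so it cannot rotate against the head, while the hair bones share tight swing limits and almost no twist.
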PhysicsAsset /BRCosmetics/Characters/Player/Male/Medium/Bodies/M_MED_GlitchPrince/Meshes/Parts/M_MED_GlitchPrince_FaceAcc_Physics
bodies {
 Crown [{shapeType=Sphere, radius=5.4797, shapeName=Crown_sphere, center=<-0.0168 -0.0081 3.9316>}]
 Hair_01 [{shapeType=Sphere, radius=2, shapeName=Hair_01_sphere, center=<5 0 0>}]
 Hair_02 [{shapeType=Sphere, radius=2, shapeName=Hair_02_sphere, center=<5 0 0>}]
 Hair_03 [{shapeType=Sphere, radius=2, shapeName=Hair_03_sphere, center=<5 0 0>}]
 Hair_04 [{shapeType=Sphere, radius=2, shapeName=Hair_04_sphere, center=<5 0 0>}]
 head [{shapeType=Capsule, length=0, radius=0.5, center=<7.9296 -0.2444 -0.4706>, rotation=<rot 0 -76.6097 -79.9487>}]
}
constraints [
 {
  bone1Name=Crown
  bone2Name=head
  swing1LimitDegrees=45
  swing1Motion=Locked
  swing2LimitDegrees=45
  swing2Motion=Locked
  twistLimitDegrees=45
  twistMotion=Locked
 }
 {bone1Name=Hair_01, bone2Name=head, swing1LimitDegrees=10, swing2LimitDegrees=20, twistLimitDegrees=1}
 {bone1Name=Hair_02, bone2Name=head, swing1LimitDegrees=10, swing2LimitDegrees=20, twistLimitDegrees=1}
 {bone1Name=Hair_03, bone2Name=head, swing1LimitDegrees=10, swing2LimitDegrees=20, twistLimitDegrees=1}
 {bone1Name=Hair_04, bone2Name=head, swing1LimitDegrees=10, swing2LimitDegrees=20, twistLimitDegrees=1}
]
