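# Anim blueprint for the GlitchPrince face accessory, copying the attached parent's pose, adding rigid body physics relative to neck_01, and animating the Crown bone.
# The event graph eases spin speed toward a constant, adds it every update without delta-time scaling, and converts it to wrapped yaw.
# That yaw additively rotates the Crown, with alpha following its absolute angle over 180 degrees.
# A second modifier offsets the Crown slightly upward in world space.
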AnimBlueprint /BRCosmetics/Characters/Player/Male/Medium/Bodies/M_MED_GlitchPrince/Meshes/Parts/M_MED_GlitchPrince_FaceAcc_AnimBP
target_skeleton=&/M_MED_GlitchPrince_FaceAcc_Skeleton
graphs [
 AnimGraph {nodes [
   n1 = RigidBody(ComponentPose=Local To Component(LocalPose=Copy Pose From Mesh() {bUseAttachedParent=true})) {
    overridePhysicsAsset=&/M_MED_GlitchPrince_FaceAcc_Physics
    simulationSpace=BaseBoneSpace
    baseBoneRef {boneName=neck_01}
   }
   n2 = "Transform (Modify) Bone - Bone: Crown"(
    ComponentPose=@n1
    Alpha=Get NewVar_0() -> (NewVar_0:double)
    Translation="0,0,0"
    Rotation="Get Spinning Rotator 00 "() -> ("Spinning Rotator 00 ":Rotator)
    Scale="1,1,1"
   ) {
    rotationMode=BMM_Additive
    rotationSpace=BCS_BoneSpace
    translationMode=BMM_Additive
    translationSpace=BCS_WorldSpace
    alphaScaleBiasClamp {interpSpeedDecreasing=2, interpSpeedIncreasing=2}
    boneToModify {boneName=Crown}
   }
   n3 = "Transform (Modify) Bone - Bone: Crown"!(ComponentPose=@n2, bAlphaBoolEnabled:bool=True, Translation="0,0,2", Rotation="0,0,0", Scale="1,1,1") -> (Pose) {
    alphaInputType=Bool
    rotationMode=BMM_Additive
    rotationSpace=BCS_BoneSpace
    translationMode=BMM_Additive
    translationSpace=BCS_WorldSpace
    alphaBoolBlend {blendInTime=1, blendOption=Sinusoidal, blendOutTime=1}
    alphaScaleBiasClamp {interpSpeedDecreasing=2, interpSpeedIncreasing=2}
    boneToModify {boneName=Crown}
    translation {z=2}
   }
   Output Pose(Result=Component To Local(ComponentPose=@n3)) {layerGroup=None}
  ]}
 EventGraph {nodes [
   n1 = Event BlueprintUpdateAnimation()
   > n2 = Sequence!(execute) -> (then_0, then_1, then_2:Exec)
   Set Crack Spin Speed(
    execute=@n2.then_1
    Crack Spin Speed:double=FInterpTo(Current=Get Crack Spin Speed() -> (Crack Spin Speed:double), Target=5, DeltaTime=@n1.DeltaTimeX, InterpSpeed=5)
   ) -> (Output_Get:double)
   n3 = Set Crack Spin(
    execute=@n2.then_2
    Crack Spin:double=float + float(A=Get Crack Spin() -> (Crack Spin:double), B=Get Crack Spin Speed()) {is=title K2Node_CommutativeAssociativeBinaryOperator}
   ) -> (Output_Get:double)
   n4 = NormalizeAxis!(Angle:float=@n3) -> (ReturnValue:float)
   > "Set Spinning Rotator 00 "("Spinning Rotator 00 ":Rotator=MakeRotator(Yaw=@n4)) -> (Output_Get:Rotator)
   n5 = Remap Float(Input=@n4, OldRange=(X=0,Y=180), NewRange=(X=0,Y=1), Clamp=true) -> (Output=~)
   n6 = Remap Float(Input=@n4, OldRange=(X=-180,Y=0), NewRange=(X=1,Y=0), Clamp=true) -> (Output=~)
   > n7 = Branch(Condition="float > float"(A=@n5, B=@n6))
   > Set NewVar_0(NewVar_0:double=@n5) -> (Output_Get:double)
   > Set NewVar_1(NewVar_1:bool=true) -> (Output_Get:bool)
   Set NewVar_0(execute=@n7.else, NewVar_0=@n6)
   > Set NewVar_1(NewVar_1=false)
  ]}
]
variables [
 Crack Spin {type=number}
 Crack Spin Speed {type=number}
 NewVar_0 {type=number}
 NewVar_1 {type=boolean}
 "Spinning Rotator 00 " {type=Rotator}
]
